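# Physics asset for the two-handed Shiny Hammer pickaxe, simulating its dangling chains and "brand" pieces.
# Boxes form the handle body, from which left and right box chains hang on wide-swing, narrow-twist joints.
# Both chains join at the last left link, closing a loop, with the right-side joint locked in all rotation.
# That shared link carries three two-segment capsule strands with locked twist, their root joints limited to one swing axis.
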PhysicsAsset /BRCosmetics/Weapons/FORT_Melee/Pickaxe_ShinyHammer/Meshes/Pickaxe_Shiny_Hammer_Pyn_Nocllosion
bodies {
 Melee_TwoHanded_Handle [
  {shapeType=Box, extentX=17.0455, extentY=10.4209, extentZ=25.1612, center=<-0.0001 8.0061 89.6225>}
  {shapeType=Box, extentX=17.0455, extentY=10.4209, extentZ=25.1612, center=<0.0402 -10.3822 89.6222>}
  {shapeType=Box, extentX=5.0768, extentY=5.0862, extentZ=36.9618, center=<1.2838 -1.0097 67.0203>}
 ]
 dyn_brand_a01 [{shapeType=Capsule, length=2, radius=1, center=<1 0 0>, rotation=<rot 90 0 0>}]
 dyn_brand_a02 [{shapeType=Capsule, length=2, radius=1, center=<2 0 0>, rotation=<rot 90 0 0>}]
 dyn_brand_b01 [{shapeType=Capsule, length=2, radius=1, center=<1 0 0>, rotation=<rot 90 0 0>}]
 dyn_brand_b02 [{shapeType=Capsule, length=2, radius=1, center=<2 0 0>, rotation=<rot 90 0 0>}]
 dyn_brand_c01 [{shapeType=Capsule, length=2, radius=1, center=<1 0 0>, rotation=<rot 90 0 0>}]
 dyn_brand_c02 [{shapeType=Capsule, length=2, radius=1, center=<2 0 0>, rotation=<rot 90 0 0>}]
 dyn_l_chain_01 [{shapeType=Box, extentX=5, extentY=1.5, extentZ=2.2561, center=<3 0 0>}]
 dyn_l_chain_02 [{shapeType=Box, extentX=5, extentY=1.5, extentZ=2.2561, center=<3.0365 0 0>, rotation=<rot 0 -5.0004 -15.0017>}]
 dyn_l_chain_03 [{shapeType=Box, extentX=7.5, extentY=2.3, extentZ=7.5}]
 dyn_r_chain_01 [{shapeType=Box, extentX=5, extentY=1.5, extentZ=2.2561, center=<3 0 0>}]
 dyn_r_chain_02 [{shapeType=Box, extentX=5, extentY=1.5, extentZ=2.2561, center=<3 0 0>, rotation=<rot 0 0 54.9994>}]
}
constraints [
 {bone1Name=dyn_r_chain_01, bone2Name=Melee_TwoHanded_Handle, swing1LimitDegrees=75, swing2LimitDegrees=75, twistLimitDegrees=5}
 {bone1Name=dyn_l_chain_01, bone2Name=Melee_TwoHanded_Handle, swing1LimitDegrees=75, swing2LimitDegrees=75, twistLimitDegrees=5}
 {
  bone1Name=dyn_brand_a01
  bone2Name=dyn_l_chain_03
  swing1LimitDegrees=30
  swing2LimitDegrees=30
  swing2Motion=Locked
  twistLimitDegrees=5
  twistMotion=Locked
 }
 {
  bone1Name=dyn_brand_b01
  bone2Name=dyn_l_chain_03
  swing1LimitDegrees=30
  swing2LimitDegrees=30
  swing2Motion=Locked
  twistLimitDegrees=5
  twistMotion=Locked
 }
 {
  bone1Name=dyn_brand_c01
  bone2Name=dyn_l_chain_03
  swing1LimitDegrees=30
  swing2LimitDegrees=30
  swing2Motion=Locked
  twistLimitDegrees=5
  twistMotion=Locked
 }
 {bone1Name=dyn_brand_a02, bone2Name=dyn_brand_a01, swing1LimitDegrees=30, swing2LimitDegrees=30, twistLimitDegrees=5, twistMotion=Locked}
 {bone1Name=dyn_brand_b02, bone2Name=dyn_brand_b01, swing1LimitDegrees=30, swing2LimitDegrees=30, twistLimitDegrees=5, twistMotion=Locked}
 {bone1Name=dyn_brand_c02, bone2Name=dyn_brand_c01, swing1LimitDegrees=30, swing2LimitDegrees=30, twistLimitDegrees=5, twistMotion=Locked}
 {bone1Name=dyn_r_chain_02, bone2Name=dyn_r_chain_01, swing1LimitDegrees=75, swing2LimitDegrees=75, twistLimitDegrees=5}
 {bone1Name=dyn_l_chain_02, bone2Name=dyn_l_chain_01, swing1LimitDegrees=75, swing2LimitDegrees=75, twistLimitDegrees=5}
 {
  bone1Name=dyn_l_chain_03
  bone2Name=dyn_l_chain_02
  swing1LimitDegrees=2
  swing1Motion=Locked
  swing2LimitDegrees=25
  twistLimitDegrees=2
  twistMotion=Locked
 }
 {
  bone1Name=dyn_l_chain_03
  bone2Name=dyn_r_chain_02
  swing1LimitDegrees=45
  swing1Motion=Locked
  swing2LimitDegrees=45
  swing2Motion=Locked
  twistLimitDegrees=45
  twistMotion=Locked
 }
]
